(event EventBeginPlay)

(event Collision|EventActorBeginOverlap
  (params (OtherActor :type Actor%20Object%20Reference)))

(event EventTick
  (params (DeltaSeconds :type Float%20%28single-precision%29)))

(fn ConstructionScript
  (Networking|CanExecuteCosmeticEvents
    (:True
      (Rendering|SetBoundsScale (Variables|Default|GetSplineMesh) (Variables|Default|GetWidth))
      (Rendering|Material|SetDefaultCustomPrimitiveDataFloat (Variables|Default|GetSplineMesh) 0 (Variables|Default|GetWidth))
      (Transformation|SetRelativeScale3D (Variables|Default|GetSplineMesh) (Math|Vector|MakeVector (/ (Variables|Default|GetHeight) 1e+02) 1.0 1.0))
      (Rendering|Material|SetDefaultCustomPrimitiveDataVector4 (Variables|Default|GetSplineMesh) 1 (Class|CreativeColorComponentBlueprintLibrary|RGBAToVec4 (Variables|Default|GetColor%20%28Bottom%29)))
      (Rendering|Material|SetDefaultCustomPrimitiveDataVector4 (Variables|Default|GetSplineMesh) 9 (Class|CreativeColorComponentBlueprintLibrary|RGBAToVec4 (Variables|Default|GetColor%20%28Top%29)))
      (Rendering|Material|SetDefaultCustomPrimitiveDataFloat (Variables|Default|GetSplineMesh) 5 (Variables|Default|GetCone%20Hardness))
      (Rendering|Material|SetDefaultCustomPrimitiveDataFloat (Variables|Default|GetSplineMesh) 6 (Variables|Default|GetCone%20Falloff%20Exponent))
      (Rendering|Material|SetDefaultCustomPrimitiveDataFloat (Variables|Default|GetSplineMesh) 7 (Variables|Default|GetLightFalloffHardness))
      (Rendering|Material|SetDefaultCustomPrimitiveDataFloat (Variables|Default|GetSplineMesh) 8 (Variables|Default|GetLightFalloffExponent))
      (Rendering|Material|SetDefaultCustomPrimitiveDataFloat (Variables|Default|GetSplineMesh) 9 (Variables|Default|GetBaseFalloffHardness))
      (Rendering|Material|SetDefaultCustomPrimitiveDataFloat (Variables|Default|GetSplineMesh) 13 (Variables|Default|GetRadiusTop))
      (Rendering|Material|SetDefaultCustomPrimitiveDataFloat (Variables|Default|GetSplineMesh) 14 (Variables|Default|GetRadiusBottom))
      (Rendering|Material|SetDefaultCustomPrimitiveDataFloat (Variables|Default|GetSplineMesh) 15 (Variables|Default|GetAngleFadeExponent)))))
