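# Additive unlit material for NeonCity fake volumetric light beams, by default visible only at night.
# UV sphere masks shape cone and length falloff, blend top and bottom colors, and combine with depth, near-camera and optional view-angle fades.
# SplineThicken widens the mesh toward the camera, while world position offset collapses it to the actor origin when visibility masks approach zero.
# Custom primitive data drives per-instance color, width and shape.
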Material /Game/Environments/Asteria/Sets/NeonCity/Techart/M_NeonCity_VolumeLight
graph {
 properties {
  bEnableResponsiveAA=true
  blendMode=BLEND_Additive
  pixelDepthOffsetMode=PDOM_Legacy
  shadingModel=MSM_Unlit
  translucencyLightingMode=TLM_VolumetricPerVertexDirectional
  usedShadingModels=Unlit
 }
 nodes [
  n1 = CollectionParameter() {
   collection=&FortniteMaterialParameters
   parameterName=FortniteActiveTimeOfDay
  }
  n2 = Constant4Vector() {constant=<rgba 0 0 0 1>}
  n3 = Max(
   A=StaticSwitchParameter(True=DotProduct(A=@n1, B=@n2.RGBA), False=Constant()) {defaultValue=true, parameterName=OnlyVisibleAtNight}
   B=NamedRerouteUsage()
  )
  n4 = NamedRerouteUsage()
  n5 = ScalarParameter() {bUseCustomPrimitiveData=true, defaultValue=100, parameterName=Width}
  If(
   A=StaticSwitchParameter(True=Min(A=@n3, B=@n4), False=@n3) {parameterName=UseAngleFadeMask}
   "A > B"=MaterialFunctionCall(WidthBase=@n5, WidthTip=@n5) {materialFunction=&SplineThicken}.WorldPositionOffset
   "A < B"=Subtract(A=ActorPositionWS(), B=WorldPosition().XYZ)
  ) {constB=0.01}
  n6 = VectorParameter() {bUseCustomPrimitiveData=true, parameterName=EmissiveTop, primitiveDataIndex=9, defaultValue=<rgba 1 0 0 0>}
  n7 = VectorParameter() {bUseCustomPrimitiveData=true, parameterName=EmissiveBottom, primitiveDataIndex=1, defaultValue=<rgba 1 0 0 0>}
  n8 = ComponentMask(None=TextureCoordinate()) {g=true}
  n9 = Constant() {r=1}
  n10 = Power(
   Base=SphereMask(
    A=@n8
    B=@n9
    Radius=@n9
    Hardness=ScalarParameter() {bUseCustomPrimitiveData=true, defaultValue=2, parameterName=LightFalloffHardness, primitiveDataIndex=7}
   )
   Exp=ScalarParameter() {bUseCustomPrimitiveData=true, defaultValue=2, parameterName=LightFalloffHardness, primitiveDataIndex=8}
  )
  LinearInterpolate(A=@n6.RGB, B=@n7.RGB, Alpha=@n10)
  n11 = Subtract(A=ActorPositionWS(), B=CameraPositionWS())
  n12 = ComponentMask(None=@n11) {g=true, r=true}
  n13 = ComponentMask(None=MaterialFunctionCall() {materialFunction=&CameraDirectionVector}.Result) {g=true, r=true}
  n14 = Power(
   Base=DotProduct(A=Normalize(VectorInput=@n12), B=Normalize(VectorInput=@n13))
   Exp=ScalarParameter() {bUseCustomPrimitiveData=true, defaultValue=16, parameterName=AngleFadeExponent, primitiveDataIndex=15}
  ) {constExponent=8}
  n15 = DotProduct(A=Multiply(A=Normalize(VectorInput=@n11)) {constB=-1}, B=Constant3Vector() {constant=<rgba 0 0 1 0>})
  NamedRerouteDeclaration(None=Saturate(None=Min(A=@n14, B=@n15))) {name=AngleFadeMaskResult, nodeColor=<rgba 1 0 0.7266 1>}
  n16 = TextureCoordinate()
  n17 = ScalarParameter() {bUseCustomPrimitiveData=true, defaultValue=0.5, parameterName=RadiusTop, primitiveDataIndex=13}
  n18 = ScalarParameter() {bUseCustomPrimitiveData=true, defaultValue=0.5, parameterName=RadiusBottom, primitiveDataIndex=14}
  n19 = Power(
   Base=SphereMask(
    A=ComponentMask(None=@n16) {r=true}
    B=Constant() {r=0.5}
    Radius=LinearInterpolate(A=@n17, B=@n18, Alpha=@n10)
    Hardness=ScalarParameter() {bUseCustomPrimitiveData=true, parameterName=Cone Hardness, primitiveDataIndex=5}
   )
   Exp=ScalarParameter() {bUseCustomPrimitiveData=true, defaultValue=2, parameterName=Cone Exponent, primitiveDataIndex=6}
  )
  n20 = Multiply(
   A=@n10
   B=SphereMask(A=@n8, B=Constant(), Hardness=ScalarParameter() {defaultValue=0.8, parameterName=BaseFalloffHardness}) {attenuationRadius=1}
  )
  n21 = MaterialFunctionCall() {materialFunction=&ObjectPivotPoint}
  n22 = CameraPositionWS()
  n23 = Saturate(None=InverseLinearInterpolate(
    A=ScalarParameter() {defaultValue=1500, parameterName=NearCameraFadeDistanceEnd}
    B=ScalarParameter() {defaultValue=5000, parameterName=NearCameraFadeDistanceStart}
    Value=Distance(A=@n21.Object Pivot Location, B=@n22)
   ))
  n24 = MaterialFunctionCall(Distance=ScalarParameter() {defaultValue=5000, parameterName=NearCameraFadeDistance}) {materialFunction=&Near_Camera_Fade}
  n25 = Multiply(
   A=DepthFade(Opacity=Multiply(A=@n19, B=@n20), FadeDistance=ScalarParameter() {defaultValue=500, parameterName=DepthFadeDistance}) {fadeDistanceDefault=500}
   B=NamedRerouteDeclaration(None=StaticSwitchParameter(True=@n23, False=@n24.Result) {parameterName=UseAdvancedNearCameraFade}) {name=NearCameraFadeResult, nodeColor=<rgba 0 0.3437 1 1>}
  )
  n26 = CollectionParameter() {
   collection=&FortniteMaterialParameters
   parameterName=FortniteActiveTimeOfDay
  }
  n27 = StaticSwitchParameter(True=Multiply(A=@n25, B=ComponentMask(None=@n26) {a=true}), False=@n25) {defaultValue=true, parameterName=OnlyVisibleAtNight}
  StaticSwitchParameter(True=Multiply(A=@n27, B=NamedRerouteUsage()), False=@n27) {parameterName=UseAngleFadeMask}
 ]
}
