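# Material function for the NeonCity set that picks one of up to four Vector4 inputs per pixel by a scalar layer index.
# The index defaults to the U of texture coordinate 1 unless UVOverride is supplied, and chained If nodes compare it against whole-number thresholds.
# Static switches Use 2, 3 and 4 Layers in the Texture Data Asset group set the chain's depth, falling back to the first input alone.
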MaterialFunction /Game/Environments/Asteria/Sets/NeonCity/Techart/MF_NeonCity_4Layer_Vec4
graph {nodes [
  n1 = Constant() {r=3}
  n2 = FunctionInput(Preview=ComponentMask(None=TextureCoordinate() {coordinateIndex=1}) {r=true}) {bUsePreviewValueAsDefault=true, inputName=UVOverride, inputType=FunctionInput_Scalar, sortPriority=4}
  n3 = FunctionInput() {bUsePreviewValueAsDefault=true, inputName=1, inputType=FunctionInput_Vector4}
  n4 = If(
   A=Constant() {r=1}
   B=@n2
   "A > B"=@n3
   "A < B"=FunctionInput() {bUsePreviewValueAsDefault=true, inputName=2, inputType=FunctionInput_Vector4, sortPriority=1}
  ) {desc=0}
  n5 = If(
   A=Constant() {r=2}
   B=@n2
   "A > B"=@n4
   "A < B"=FunctionInput() {bUsePreviewValueAsDefault=true, inputName=3, inputType=FunctionInput_Vector4, sortPriority=2}
  )
  n6 = FunctionInput() {bUsePreviewValueAsDefault=true, inputName=4, inputType=FunctionInput_Vector4, sortPriority=3}
  n7 = StaticSwitch(True=@n4, False=@n3, Value=StaticBoolParameter() {group=Texture Data Asset, parameterName=Use 2 Layers})
  n8 = StaticBoolParameter() {group=Texture Data Asset, parameterName=Use 3 Layers}
  FunctionOutput(None=StaticSwitch(
    True=If(A=@n1, B=@n2, "A > B"=@n5, "A < B"=@n6)
    False=StaticSwitch(True=@n5, False=@n7, Value=@n8)
    Value=StaticBoolParameter() {group=Texture Data Asset, parameterName=Use 4 Layers}
   ))
 ]}
